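# Niagara idle effect for the ArcticIceBlue pickaxe, adding blue glow and crackling electricity.
# Glow001 pins sprites to triangles of a glow mesh and emits a time-of-day-scaled light, with a Timeline pulse driven by reading Impact_Lightning.
# Impact_Lightning bursts short-lived ribbon arcs along a curve-shaped cylinder path, flickered by stepped width curves.
# Lightning_Blade emitters scatter mesh sparks with sprites and lights; platform quality scales glow and arc particle counts.
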NiagaraSystem /BRCosmetics/Weapons/FORT_Melee/PIckaxe_ArcticIceBlue/FX/NS_ArcticIceBlue_Idle
dependencies {
 usedDataInterfaces [
  &NiagaraDataInterfaceMeshRendererInfo
  &NiagaraDataInterfaceCurve
  &NiagaraDataInterfaceParticleRead
  &NiagaraDataInterfacePlatformSet
  &NiagaraDataInterfaceStaticMesh
 ]
 usedDynamicInputs []
 usedModules [&SystemState]
}
emitters {
 Glow001 {
  emitterUpdateScript [
   &EmitterState()
   &SpawnBurst_Instantaneous(Spawn Count=183, Use Spawn Probability=false, UseLoopCountLimit=true)
   &Timeline(
    Instantaneous Rewind=true
    Play=&self:CheckOhterEmitter(New Particle Attribute Reader:NiagaraDataInterfaceParticleRead=DataInterface() {propertyValues {EmitterBinding {bindingMode=Other, emitterName=Impact_Lightning}}})
    Play Rate=&Divide_Float(A=1, B=0.3)
    Rewind Rate=2.5
   )
  ]
  particleSpawnScript [
   &InitializeParticle(
    Color=<rgba 0.1779 0.3916 0.8148 0.4>
    Lifetime=0.25
    Sprite Size=&Multiply_Vector2D_ByFloat(Float=0.4, Vector2D=<35 80>)
    Sprite Size Mode=::NewEnumerator4
    UsePositionOffset=false
    Write Material Random=true
   )
   &CameraOffset(Camera Offset Amount=-5)
   &KillParticles(Kill Particles=&SetBoolByIntComparison(
     A=&ModuloInt(
      A=&ReturnParticleIDAttributes(Particle ID Attribute=::NewEnumerator1)
      B=&MakeCustomIntFromBool(
       Bool=&IsPlatformSetActive(New Platform Set=DataInterface() {propertyValues {Platforms {qualityLevelMask=-14}}})
       True INT=6
       False INT=&MakeCustomIntFromBool(
        Bool=&IsPlatformSetActive(New Platform Set=DataInterface() {propertyValues {Platforms {qualityLevelMask=-28}}})
        True INT=5
        False INT=4
       )
      )
     )
     B=0
     Comparison Type=::NewEnumerator0
    ))
   &KillParticles(Kill Particles=&SetBoolByIntComparison(
     A=&ReturnParticleIDAttributes(Particle ID Attribute=::NewEnumerator1)
     B=140
     Comparison Type=::NewEnumerator4
    )) {module=KillParticles001}
  ]
  particleUpdateScript [
   &ParticleState(Kill Particles When Lifetime Has Elapsed=false, Let Infinitely Lived Particles Die When Emitter Deactivates=true)
   /Niagara/Modules/Spawn/Location/StaticMeshLocation(
    FixedBarycentric=true
    MaskSection=false
    MaskSpawnGroup=false
    Offset Position Along Sampled Normal=0
    OffsetAlongNormal=true
    SamplePosOffset=true
    Static Mesh=DataInterface() {propertyValues {
      DefaultMesh=&/SM_ArcticIcePickGlow
      InstanceIndex=-1
      LODIndex=0
      PreviewMesh=&SM_ArcticIcePickGlow
      SoftSourceActor=null
      SourceMode=DefaultMeshOnly
      bAllowSamplingFromStreamingLODs=false
      bCaptureTransformsPerFrame=true
      bUsePhysicsBodyVelocity=false
      FilteredSockets []
      LODIndexUserParameter {parameter {name=None, typeDefHandle {registeredTypeIndex=108}, varData []}}
      MeshParameterBinding {parameter {name=None, typeDefHandle {registeredTypeIndex=1}, varData []}}
      SectionFilter {allowedMaterialSlots []}
     }}
    Triangle ID:NiagaraInt32=&ReturnParticleIDAttributes(Particle ID Attribute=::NewEnumerator1)
    Triangle Sampling Mode:ENiagara_StaticTriangleMode=::NewEnumerator2
    Write To Particle Position=false
   )
   /SetVariables(Particles.OffPosition:NiagaraPosition=&AddVectorToPosition(
     Position=$Particles.StaticMeshLocation.SampledPosition
     Vector=&Multiply_VectorByFloat(
      Vector=&Lerp_Vector(
       A=$Output.StaticMeshLocation.SampledTangent
       B=$Output.StaticMeshLocation.SampledBitangent
       Alpha=&MakeCustomFloatFromBool(
        Bool=&SetBoolByIntComparison(
         A=&ReturnParticleIDAttributes(Particle ID Attribute=::NewEnumerator1)
         B=95
         Comparison Type=::NewEnumerator0
        )
        True Float=1
        False Float=0
       )
      )
      Float=-3.6011
     )
    )) {isSetParameters=true}
   &ScaleColor(
    Scale Alpha=&Multiply_Float(
     A=&Lerp_Float(A=0.5, B=1, Alpha=&Sine(Normalized Angle=$Emitter.Age, Period=3, Scale=0.5, Bias=0.5))
     B=&Multiply_Float(A=&FNTimeOfDayScalar(Night Scale=0.75, Morning Scale=1, Evening Scale=1), B=1)
    )
    ScaleA=true
    ScaleRGB=true
   )
   /SetVariables(Particles.Velocity=&Lerp_Vector(
     A=$Output.StaticMeshLocation.SampledTangent
     B=$Output.StaticMeshLocation.SampledBitangent
     Alpha=&MakeCustomFloatFromBool(
      Bool=&SetBoolByIntComparison(
       A=&ReturnParticleIDAttributes(Particle ID Attribute=::NewEnumerator1)
       B=95
       Comparison Type=::NewEnumerator0
      )
      True Float=1
      False Float=0
     )
    )) {isSetParameters=true}
   &Color(Color=&LerpLinearColors(
     StartColor=<rgba 1 1 1 1>
     EndColor=<rgba 0 0 0 1>
     LerpFactor=&MakeFloatFromInt(Int=&MakeCustomIntFromBool(
       Bool=&SetBoolByIntComparison(A=&ReturnParticleIDAttributes(), B=74, Comparison Type=::NewEnumerator4)
       True INT=0
       False INT=1
      ))
    )) {enabled=false}
   &ScaleColor(
    Scale RGB=&VectorFromFloat(Value=&FloatFromCurve(
      FloatCurve=DataInterface() {propertyValues {Curve {keys [
          {interpMode=RCIM_Cubic, tangentMode=RCTM_User, time=0, value=1}
          {interpMode=RCIM_Cubic, tangentMode=RCTM_User, time=0.1, value=2}
          {
           arriveTangentWeight=0.7721
           interpMode=RCIM_Cubic
           leaveTangentWeight=0.7721
           tangentMode=RCTM_User
           tangentWeightMode=RCTWM_WeightedBoth
           time=1
           value=1
          }
         ]}}}
      CurveIndex=$Emitter.Timeline.NormalizedTime
      Scale Curve=2.5
     ))
    ScaleA=true
    ScaleRGB=true
   ) {module=ScaleColor001}
   &FN_Light_Attributes(
    Color=$Particles.Color
    Create Color Attribute=true
    Create Exponent Attribute=true
    Create Position Attribute=true
    Create Radius Attribute=true
    Enable Particle Light Enable=true
    Exponent=3
    Light Enabled:NiagaraBool=&SetBoolByIntComparison(
     A=&ReturnParticleIDAttributes(Particle ID Attribute=::NewEnumerator1)
     B=0
     Comparison Type=::NewEnumerator4
    )
    Position=&AddVectorToPosition(
     Position=/Niagara/DynamicInputs/Helpers/SimulationPosition()
     Vector=&TransformVector(Vector=<15 5 30>, Destination Space=::Simulation)
    )
    Radius=55
    Scale Color=&FNTimeOfDayScalar(
     Night Scale=&FloatFromCurve(
      FloatCurve=DataInterface() {propertyValues {Curve {keys [
          {interpMode=RCIM_Cubic, tangentMode=RCTM_User, time=0, value=3.45}
          {interpMode=RCIM_Cubic, tangentMode=RCTM_User, time=0.1, value=3.7}
          {
           arriveTangentWeight=0.7721
           interpMode=RCIM_Cubic
           leaveTangentWeight=0.7721
           tangentMode=RCTM_User
           tangentWeightMode=RCTWM_WeightedBoth
           time=1
           value=3.45
          }
         ]}}}
      CurveIndex=$Emitter.Timeline.NormalizedTime
     )
     Morning Scale=&FloatFromCurve(
      FloatCurve=DataInterface() {propertyValues {Curve {keys [
          {interpMode=RCIM_Cubic, tangentMode=RCTM_User, time=0, value=1}
          {interpMode=RCIM_Cubic, tangentMode=RCTM_User, time=0.1, value=2}
          {
           arriveTangentWeight=0.7721
           interpMode=RCIM_Cubic
           leaveTangentWeight=0.7721
           tangentMode=RCTM_User
           tangentWeightMode=RCTWM_WeightedBoth
           time=1
           value=1
          }
         ]}}}
      CurveIndex=$Emitter.Timeline.NormalizedTime
     )
     Day Scale=&FloatFromCurve(
      FloatCurve=DataInterface() {propertyValues {Curve {keys [
          {interpMode=RCIM_Cubic, tangentMode=RCTM_User, time=0, value=2.6}
          {interpMode=RCIM_Cubic, tangentMode=RCTM_User, time=0.1, value=2.8}
          {
           arriveTangentWeight=0.7721
           interpMode=RCIM_Cubic
           leaveTangentWeight=0.7721
           tangentMode=RCTM_User
           tangentWeightMode=RCTWM_WeightedBoth
           time=1
           value=2.6
          }
         ]}}}
      CurveIndex=$Emitter.Timeline.NormalizedTime
     )
     Evening Scale=&FloatFromCurve(
      FloatCurve=DataInterface() {propertyValues {Curve {keys [
          {interpMode=RCIM_Cubic, tangentMode=RCTM_User, time=0, value=1}
          {interpMode=RCIM_Cubic, tangentMode=RCTM_User, time=0.1, value=2}
          {
           arriveTangentWeight=0.7721
           interpMode=RCIM_Cubic
           leaveTangentWeight=0.7721
           tangentMode=RCTM_User
           tangentWeightMode=RCTWM_WeightedBoth
           time=1
           value=1
          }
         ]}}}
      CurveIndex=$Emitter.Timeline.NormalizedTime
     )
    )
   )
  ]
  rendererClasses [&NiagaraSpriteRendererProperties, &NiagaraLightRendererProperties]
  renderers [
   {rendererClass=&NiagaraSpriteRendererProperties}
   {rendererClass=&NiagaraLightRendererProperties, rendererIndex=1}
  ]
 }
 Impact_Lightning {
  emitterSpawnScript [/SetVariables(Emitter.New Curve for Floats=DataInterface() {propertyValues {Curve {keys [
        {interpMode=RCIM_Constant, leaveTangentWeight=0.003, time=0, value=0}
        {arriveTangentWeight=0.003, interpMode=RCIM_Constant, leaveTangentWeight=0.0092, time=0.0091, value=1}
        {arriveTangentWeight=0.0092, interpMode=RCIM_Constant, leaveTangentWeight=0.0066, time=0.0366, value=0}
        {arriveTangentWeight=0.0066, interpMode=RCIM_Constant, leaveTangentWeight=0.002, time=0.0566, value=1}
        {arriveTangentWeight=0.002, interpMode=RCIM_Constant, leaveTangentWeight=0.0051, time=0.0627, value=0}
        {arriveTangentWeight=0.0051, interpMode=RCIM_Constant, leaveTangentWeight=0.0118, time=0.078, value=1}
        {arriveTangentWeight=0.0118, interpMode=RCIM_Constant, leaveTangentWeight=0.0107, time=0.1133, value=0}
        {arriveTangentWeight=0.0107, interpMode=RCIM_Constant, leaveTangentWeight=0.0077, time=0.1454, value=1}
        {arriveTangentWeight=0.0077, interpMode=RCIM_Constant, leaveTangentWeight=0.003, time=0.1684, value=0}
        {arriveTangentWeight=0.003, interpMode=RCIM_Constant, leaveTangentWeight=0.0092, time=0.1775, value=1}
        {arriveTangentWeight=0.0092, interpMode=RCIM_Constant, leaveTangentWeight=0.0112, time=0.2051, value=0}
        {arriveTangentWeight=0.0112, interpMode=RCIM_Constant, leaveTangentWeight=0.0264, time=0.2386, value=1}
        {arriveTangentWeight=0.0264, interpMode=RCIM_Constant, leaveTangentWeight=0.0309, time=0.3178, value=0}
        {arriveTangentWeight=0.0309, interpMode=RCIM_Constant, leaveTangentWeight=0.0307, time=0.4104, value=1}
        {arriveTangentWeight=0.1965, interpMode=RCIM_Constant, leaveTangentWeight=0.141, time=0.5027, value=0}
        {arriveTangentWeight=0.2939, interpMode=RCIM_Constant, leaveTangentWeight=0.0404, time=0.5689, value=0}
        {arriveTangentWeight=0.003, interpMode=RCIM_Constant, leaveTangentWeight=0.0092, time=0.578, value=1}
        {arriveTangentWeight=0.1715, interpMode=RCIM_Constant, leaveTangentWeight=0.1623, time=0.6055, value=0}
        {arriveTangentWeight=0.21, interpMode=RCIM_Constant, leaveTangentWeight=0.1236, time=0.6315, value=0}
        {arriveTangentWeight=0.0051, interpMode=RCIM_Constant, leaveTangentWeight=0.0118, time=0.6468, value=1}
        {arriveTangentWeight=0.1308, interpMode=RCIM_Constant, leaveTangentWeight=0.2039, time=0.6822, value=0}
        {arriveTangentWeight=0.2867, interpMode=RCIM_Constant, leaveTangentWeight=0.0474, time=0.7373, value=0}
        {arriveTangentWeight=0.003, interpMode=RCIM_Constant, leaveTangentWeight=0.0092, time=0.7464, value=1}
        {arriveTangentWeight=0.0092, interpMode=RCIM_Constant, leaveTangentWeight=0.0112, time=0.7739, value=0}
        {arriveTangentWeight=0.0112, interpMode=RCIM_Constant, leaveTangentWeight=0.0264, time=0.8075, value=1}
        {arriveTangentWeight=0.0264, interpMode=RCIM_Constant, leaveTangentWeight=0.003, time=0.8867, value=0}
        {arriveTangentWeight=0.003, interpMode=RCIM_Constant, leaveTangentWeight=0.0092, time=0.8958, value=1}
        {arriveTangentWeight=0.1715, interpMode=RCIM_Constant, leaveTangentWeight=0.1623, time=0.9233, value=0}
        {arriveTangentWeight=0.2587, interpMode=RCIM_Constant, leaveTangentWeight=0.0748, time=0.9493, value=0}
        {arriveTangentWeight=0.0025, interpMode=RCIM_Constant, leaveTangentWeight=0.0144, time=0.9568, value=1}
        {arriveTangentWeight=0.0144, interpMode=RCIM_Constant, time=1, value=0}
       ]}}}) {isSetParameters=true}]
  emitterUpdateScript [
   &EmitterState(
    Enable Distance Culling=true
    Life Cycle Mode=::NewEnumerator1
    Loop Duration=&RandomRangeFloat(Minimum=0.25, Maximum=2.5)
    MaxDistance=3500
    Module.bUseMaxDistance=true
    Module.bUseMinDistance=false
    Recalculate Duration Each Loop=true
    Scalability Mode=::NewEnumerator1
    UseLoopDelay=false
   )
   /SetVariables(
    Emitter.Rand1:NiagaraFloat=&RandomRangeFloat(Minimum=0, Maximum=1, Evaluation Type=::NewEnumerator1)
    Emitter.Rand2:NiagaraFloat=&RandomRangeFloat(Minimum=0, Maximum=1, Evaluation Type=::NewEnumerator1)
   ) {isSetParameters=true}
   /SetVariables(Emitter.Electric Particles:NiagaraInt32=&MakeCustomIntFromBool(
     Bool=&IsPlatformSetActive(New Platform Set=DataInterface() {propertyValues {Platforms {qualityLevelMask=-31}}})
     True INT=15
     False INT=&MakeCustomIntFromBool(
      Bool=&IsPlatformSetActive(New Platform Set=DataInterface() {propertyValues {Platforms {qualityLevelMask=-26}}})
      True INT=20
      False INT=25
     )
    )) {isSetParameters=true}
   &SpawnBurst_Instantaneous(Spawn Count=$Emitter.Electric Particles, Use Spawn Probability=false, UseLoopCountLimit=false)
   /SetVariables(Emitter.Time Offset Random:NiagaraFloat=&RandomRangeFloat(Minimum=0, Maximum=1)) {isSetParameters=true}
   /SetVariables(Emitter.ribbon life:NiagaraFloat=&RandomRangeFloat(Minimum=0.175, Maximum=0.25, Evaluation Type=::NewEnumerator1)) {isSetParameters=true}
  ]
  particleSpawnScript [
   &PartitionParticles(
    Normalized Index In Partition Includes One=true
    Particles Per Partition:NiagaraInt32=$Emitter.Electric Particles
    Partition Count Mode:ENiagara_InfiniteLoopDuration=::NewEnumerator1
    Partition Mode:ENiagara_PartitionMode=::NewEnumerator1
    Write Partition Index to Ribbon Link Order:ENiagara_AttributeSamplingApplyOutput=::NewEnumerator0
    Write Partition to Ribbon ID=::NewEnumerator0
   )
   &InitializeParticle(
    Color=&RandomRangeLinearColor(Minimum=<rgba 0 4.1574 8 1>, Maximum=<rgba 0.15 1.6311 3 1>)
    Lifetime=$Emitter.ribbon life
    Ribbon Width=&Multiply_Float(A=&RandomRangeFloat(Minimum=6, Maximum=5), B=0.6)
    Ribbon Width Mode=::NewEnumerator1
    Sprite Size Mode=::NewEnumerator3
    Uniform Sprite Size=1
    UsePositionOffset=false
    Write Material Random=true
   )
   /SetVariables(
    Particles.RibbonID=/Niagara/DynamicInputs/NiagaraID/MakeNiagaraIDFromInt(Index:NiagaraInt32=$Emitter.LoopCount, Acquire Tag:NiagaraInt32=0)
    Particles.link order:NiagaraFloat=&MakeFloatFromInt(Int=&ReturnParticleIDAttributes())
   ) {isSetParameters=true}
   &CylinderLocation(
    Clamp UV Range=false
    Cylinder Distribution:ENiagara_CylinderMode=::NewEnumerator2
    Cylinder Height=150
    Cylinder Height Midpoint=0.55
    Cylinder Origin:NiagaraPosition=&SimulationPosition()
    Cylinder Radius=11
    MaskSpawnGroup=false
    Offset=<1 7 0>
    Offset Mode:ENiagara_OffsetMode=::NewEnumerator0
    Radius Position:NiagaraFloat=&FloatFromCurve(
     FloatCurve=DataInterface() {propertyValues {Curve {keys [
         {interpMode=RCIM_Cubic, tangentMode=RCTM_User, time=0, value=0.4}
         {
          arriveTangent=0.672
          arriveTangentWeight=0.1067
          interpMode=RCIM_Cubic
          leaveTangentWeight=0.3202
          tangentMode=RCTM_Break
          time=0.5
          value=0.4236
         }
         {interpMode=RCIM_Cubic, tangentMode=RCTM_User, time=0.8, value=0.6099}
         <at 0.8394 0.604>
         <at 0.8655 1.3393>
         {arriveTangentWeight=0.1415, interpMode=RCIM_Cubic, leaveTangentWeight=0.1415, tangentMode=RCTM_User, time=0.9, value=0.5328}
         <at 1 0.6132 RCIM_Cubic>
        ]}}}
     CurveIndex=$Particles.PartitionParticles.Normalized Index In Partition
    )
    U Position:NiagaraFloat=&FloatFromCurve(
     FloatCurve=DataInterface() {propertyValues {Curve {keys [{tangentMode=RCTM_User, time=0, value=2}, {tangentMode=RCTM_User, time=1, value=0.5}]}}}
     CurveIndex=$Particles.PartitionParticles.Normalized Index In Partition
    )
    UseOwnerScale=true
    V Position:NiagaraFloat=&FloatFromCurve(
     FloatCurve=DataInterface() {propertyValues {Curve {keys [{arriveTangent=1.5333, time=0, value=0.1001}, <at 0.9146 0.8469>, {tangentMode=RCTM_User, time=1, value=0.85}]}}}
     CurveIndex=$Particles.PartitionParticles.Normalized Index In Partition
    )
   )
   &CylinderLocation(
    Clamp UV Range=false
    Cylinder Distribution:ENiagara_CylinderMode=::NewEnumerator2
    Cylinder Height=150
    Cylinder Height Midpoint=0.55
    Cylinder Origin:NiagaraPosition=&SimulationPosition()
    Cylinder Radius=11
    MaskSpawnGroup=false
    Offset=<1 7 0>
    Offset Mode:ENiagara_OffsetMode=::NewEnumerator0
    Radius Position:NiagaraFloat=&FloatFromCurve(
     FloatCurve=DataInterface() {propertyValues {Curve {keys [
         {interpMode=RCIM_Cubic, tangentMode=RCTM_User, time=0, value=0.4}
         {
          arriveTangent=0.672
          arriveTangentWeight=0.1067
          interpMode=RCIM_Cubic
          leaveTangentWeight=0.3202
          tangentMode=RCTM_Break
          time=0.5
          value=0.4236
         }
         {interpMode=RCIM_Cubic, tangentMode=RCTM_User, time=0.8, value=0.6099}
         <at 0.8394 0.604>
         <at 0.8655 1.3393>
         {arriveTangentWeight=0.1415, interpMode=RCIM_Cubic, leaveTangentWeight=0.1415, tangentMode=RCTM_User, time=0.9, value=0.5328}
         <at 1 1.225 RCIM_Cubic>
        ]}}}
     CurveIndex=$Particles.PartitionParticles.Normalized Index In Partition
    )
    U Position:NiagaraFloat=&FloatFromCurve(
     FloatCurve=DataInterface() {propertyValues {Curve {keys [
         {arriveTangent=6.2373, interpMode=RCIM_Cubic, leaveTangent=6.2373, tangentMode=RCTM_User, time=0, value=0.3596}
         {arriveTangent=2.109, interpMode=RCIM_Cubic, leaveTangent=2.109, time=0.1949, value=1.3086}
         {
          arriveTangentWeight=0.3586
          interpMode=RCIM_Cubic
          leaveTangent=-3.4745
          leaveTangentWeight=0.3541
          tangentMode=RCTM_Break
          time=0.6
          value=1.625
         }
         {tangentMode=RCTM_User, time=1, value=-0.75}
        ]}}}
     CurveIndex=$Particles.PartitionParticles.Normalized Index In Partition
    )
    UseOwnerScale=true
    V Position:NiagaraFloat=&FloatFromCurve(
     FloatCurve=DataInterface() {propertyValues {Curve {keys [
         {arriveTangent=1.5333, interpMode=RCIM_Cubic, leaveTangent=1.5333, tangentMode=RCTM_User, time=0, value=0.1001}
         {arriveTangent=0.6665, interpMode=RCIM_Cubic, leaveTangent=0.6665, time=0.187, value=0.2321}
         <at 0.6 0.5>
         <at 0.7 0.6>
         <at 0.7532 0.6737>
         <at 0.8141 0.7644>
         <at 0.8648 0.8072>
         <at 0.8815 0.8803>
         <at 0.9146 0.8469>
         {tangentMode=RCTM_User, time=1, value=0.8974}
        ]}}}
     CurveIndex=$Particles.PartitionParticles.Normalized Index In Partition
    )
   ) {enabled=false, module=CylinderLocation001}
   &ShapeLocation(
    MaskSpawnGroup=false
    Offset=&MakeVector(
     X=0
     Y=&FloatFromCurve(
      FloatCurve=DataInterface() {propertyValues {Curve {keys [{tangentMode=RCTM_User, time=0, value=0}, <at 0.9 0>, {tangentMode=RCTM_User, time=1, value=1}]}}}
      CurveIndex=$Particles.PartitionParticles.Normalized Index In Partition
      Scale Curve=26.2269
     )
     Z=0
    )
    Offset Mode=::NewEnumerator0
    Override Seed=false
    Sphere Radius=0
    UseOwnerScale=true
   ) {module=ShapeLocation001}
  ]
  particleUpdateScript [
   &ParticleState()
   &ScaleColor(
    Scale Alpha=10
    Scale RGB=&VectorFromFloat(Value=&Power(a=&Waveform("Waveform [1]"=::NewEnumerator12, "[1] Frequency"=35, "[1] Amplitude Min/Max"=<0.8 1>), b=5))
    ScaleA=true
    ScaleRGB=true
   )
   &ShapeLocation(
    Evaluation Type:ENiagaraRandomnessEvaluation=::NewEnumerator1
    MaskSpawnGroup=false
    Override Seed=false
    Shape Origin=$Particles.Position
    Sphere Radius=&RandomRangeFloat(Minimum=0, Maximum=2, Evaluation Type=::NewEnumerator1)
    UseOwnerScale=true
   )
   &ScaleRibbonWidth(Ribbon Width Scale=&Multiply_Float(
     A=&Multiply_Float(
      A=&FloatFromCurve(FloatCurve=DataInterface() {propertyValues {Curve {keys [
           <at 0 0>
           {interpMode=RCIM_Cubic, tangentMode=RCTM_User, time=0.05, value=1}
           {arriveTangentWeight=0.1491, interpMode=RCIM_Cubic, leaveTangentWeight=0.2236, tangentMode=RCTM_User, time=0.8, value=1}
           <at 1 0>
          ]}}})
      B=&FloatFromCurve(
       FloatCurve=DataInterface() {propertyValues {Curve {keys [
           {arriveTangent=6.6108, interpMode=RCIM_Cubic, leaveTangent=6.6108, tangentMode=RCTM_User, time=0, value=0}
           {arriveTangent=2, interpMode=RCIM_Cubic, leaveTangent=2, time=0.05, value=0.7}
           {interpMode=RCIM_Cubic, tangentMode=RCTM_User, time=0.5, value=1}
           {arriveTangent=-6.3941, interpMode=RCIM_Cubic, leaveTangent=-6.3941, tangentMode=RCTM_User, time=1, value=0}
          ]}}}
       CurveIndex=$Particles.Initial.PartitionParticles.Normalized Index In Partition
      )
     )
     B=&FloatFromCurve(
      FloatCurve=DataInterface() {propertyValues {Curve {keys [
          {interpMode=RCIM_Constant, leaveTangentWeight=0.003, time=0, value=0}
          {arriveTangentWeight=0.003, interpMode=RCIM_Constant, leaveTangentWeight=0.0092, time=0.0091, value=1}
          {arriveTangentWeight=0.0092, interpMode=RCIM_Constant, leaveTangentWeight=0.0066, time=0.0366, value=0}
          {arriveTangentWeight=0.002, interpMode=RCIM_Constant, leaveTangentWeight=0.0051, time=0.0627, value=0}
          {arriveTangentWeight=0.0051, interpMode=RCIM_Constant, leaveTangentWeight=0.0118, time=0.078, value=1}
          {arriveTangentWeight=0.0118, interpMode=RCIM_Constant, leaveTangentWeight=0.0107, time=0.1133, value=0}
          {arriveTangentWeight=0.0077, interpMode=RCIM_Constant, leaveTangentWeight=0.003, time=0.1684, value=0}
          {arriveTangentWeight=0.003, interpMode=RCIM_Constant, leaveTangentWeight=0.0092, time=0.1775, value=1}
          {arriveTangentWeight=0.0092, interpMode=RCIM_Constant, leaveTangentWeight=0.0112, time=0.2051, value=0}
          {arriveTangentWeight=0.0112, interpMode=RCIM_Constant, leaveTangentWeight=0.0264, time=0.2386, value=1}
          {arriveTangentWeight=0.0264, interpMode=RCIM_Constant, leaveTangentWeight=0.0309, time=0.3178, value=0}
          {arriveTangentWeight=0.1723, interpMode=RCIM_Constant, leaveTangentWeight=0.1666, time=0.4104, value=1}
          {arriveTangentWeight=0.3251, interpMode=RCIM_Constant, leaveTangentWeight=0.0097, time=0.5, value=1}
          {arriveTangentWeight=0.0129, interpMode=RCIM_Constant, leaveTangentWeight=0.3212, time=0.5027, value=0}
          {arriveTangentWeight=0.2939, interpMode=RCIM_Constant, leaveTangentWeight=0.0404, time=0.5689, value=0}
          {arriveTangentWeight=0.003, interpMode=RCIM_Constant, leaveTangentWeight=0.0092, time=0.578, value=1}
          {arriveTangentWeight=0.1715, interpMode=RCIM_Constant, leaveTangentWeight=0.1623, time=0.6055, value=0}
          {arriveTangentWeight=0.21, interpMode=RCIM_Constant, leaveTangentWeight=0.1236, time=0.6315, value=0}
          {arriveTangentWeight=0.0051, interpMode=RCIM_Constant, leaveTangentWeight=0.0118, time=0.6468, value=1}
          {arriveTangentWeight=0.1308, interpMode=RCIM_Constant, leaveTangentWeight=0.2039, time=0.6822, value=0}
          {arriveTangentWeight=0.2867, interpMode=RCIM_Constant, leaveTangentWeight=0.0474, time=0.7373, value=0}
          {arriveTangentWeight=0.003, interpMode=RCIM_Constant, leaveTangentWeight=0.0092, time=0.7464, value=1}
          {arriveTangentWeight=0.0092, interpMode=RCIM_Constant, leaveTangentWeight=0.0112, time=0.7739, value=0}
          {arriveTangentWeight=0.0112, interpMode=RCIM_Constant, leaveTangentWeight=0.0264, time=0.8075, value=1}
          {arriveTangentWeight=0.0264, interpMode=RCIM_Constant, leaveTangentWeight=0.003, time=0.8867, value=0}
          {arriveTangentWeight=0.003, interpMode=RCIM_Constant, leaveTangentWeight=0.0092, time=0.8958, value=1}
          {arriveTangentWeight=0.1715, interpMode=RCIM_Constant, leaveTangentWeight=0.1623, time=0.9233, value=0}
          {arriveTangentWeight=0.2587, interpMode=RCIM_Constant, leaveTangentWeight=0.0748, time=0.9493, value=0}
          {arriveTangentWeight=0.0025, interpMode=RCIM_Constant, leaveTangentWeight=0.0144, time=0.9568, value=1}
          {arriveTangentWeight=0.0144, interpMode=RCIM_Constant, time=1, value=0}
         ]}}}
      CurveIndex=&Frac_Float(Float=&Add_Float(
        A=&Frac_Float(Float=&Add_Float(A=&Multiply_Float(A=$Particles.PartitionParticles.Normalized Index In Partition, B=-0.25), B=1))
        B=&Frac_Float(Float=&Multiply_Float(
          A=&Add_Float(
           A=&self:Pixelate(Input:NiagaraFloat=$Emitter.Age, Pixel Value:NiagaraFloat=&RandomRangeFloat(Minimum=24, Maximum=8))
           B=$Emitter.Time Offset Random
          )
          B=0.35
         ))
       ))
     )
    )) {enabled=false, module=ScaleRibbonWidth001}
   &ScaleRibbonWidth(Ribbon Width Scale=&Multiply_Float(
     A=&Multiply_Float(
      A=&FloatFromCurve(FloatCurve=DataInterface() {propertyValues {Curve {keys [
           <at 0 0>
           {interpMode=RCIM_Cubic, tangentMode=RCTM_User, time=0.05, value=1}
           {arriveTangentWeight=0.1491, interpMode=RCIM_Cubic, leaveTangentWeight=0.2236, tangentMode=RCTM_User, time=0.8, value=1}
           <at 1 0>
          ]}}})
      B=&FloatFromCurve(
       FloatCurve=DataInterface() {propertyValues {Curve {keys [
           {arriveTangent=6.6108, interpMode=RCIM_Cubic, leaveTangent=6.6108, tangentMode=RCTM_User, time=0, value=0}
           {arriveTangent=2, interpMode=RCIM_Cubic, leaveTangent=2, time=0.05, value=0.7}
           {interpMode=RCIM_Cubic, tangentMode=RCTM_User, time=0.5, value=1}
           {arriveTangent=-6.3941, interpMode=RCIM_Cubic, leaveTangent=-6.3941, tangentMode=RCTM_User, time=1, value=0}
          ]}}}
       CurveIndex=$Particles.Initial.PartitionParticles.Normalized Index In Partition
      )
     )
     B=&FloatFromCurve(
      FloatCurve=DataInterface() {propertyValues {Curve {keys [
          {interpMode=RCIM_Constant, leaveTangentWeight=0.003, time=0, value=0}
          {arriveTangentWeight=0.003, interpMode=RCIM_Constant, leaveTangentWeight=0.0092, time=0.0091, value=1}
          {arriveTangentWeight=0.0051, interpMode=RCIM_Constant, leaveTangentWeight=0.0118, time=0.078, value=1}
          {arriveTangentWeight=0.0118, interpMode=RCIM_Constant, leaveTangentWeight=0.0107, time=0.1133, value=0}
          {arriveTangentWeight=0.003, interpMode=RCIM_Constant, leaveTangentWeight=0.0092, time=0.1775, value=1}
          {arriveTangentWeight=0.0112, interpMode=RCIM_Constant, leaveTangentWeight=0.0264, time=0.2386, value=1}
          {arriveTangentWeight=0.0264, interpMode=RCIM_Constant, leaveTangentWeight=0.0309, time=0.3178, value=0}
          {arriveTangentWeight=0.1723, interpMode=RCIM_Constant, leaveTangentWeight=0.1666, time=0.4104, value=1}
          {arriveTangentWeight=0.3251, interpMode=RCIM_Constant, leaveTangentWeight=0.0097, time=0.5, value=1}
          {arriveTangentWeight=0.003, interpMode=RCIM_Constant, leaveTangentWeight=0.0092, time=0.578, value=1}
          {arriveTangentWeight=0.1715, interpMode=RCIM_Constant, leaveTangentWeight=0.1623, time=0.6055, value=0}
          {arriveTangentWeight=0.21, interpMode=RCIM_Constant, leaveTangentWeight=0.1236, time=0.6315, value=0}
          {arriveTangentWeight=0.0051, interpMode=RCIM_Constant, leaveTangentWeight=0.0118, time=0.6468, value=1}
          {arriveTangentWeight=0.003, interpMode=RCIM_Constant, leaveTangentWeight=0.0092, time=0.7464, value=1}
          {arriveTangentWeight=0.0092, interpMode=RCIM_Constant, leaveTangentWeight=0.0112, time=0.7739, value=0}
          {arriveTangentWeight=0.0112, interpMode=RCIM_Constant, leaveTangentWeight=0.0264, time=0.8075, value=1}
          {arriveTangentWeight=0.003, interpMode=RCIM_Constant, leaveTangentWeight=0.0092, time=0.8958, value=1}
          {arriveTangentWeight=0.1715, interpMode=RCIM_Constant, leaveTangentWeight=0.1623, time=0.9233, value=0}
          {arriveTangentWeight=0.2587, interpMode=RCIM_Constant, leaveTangentWeight=0.0748, time=0.9493, value=0}
          {arriveTangentWeight=0.0025, interpMode=RCIM_Constant, leaveTangentWeight=0.0144, time=0.9568, value=1}
          {arriveTangentWeight=0.0144, interpMode=RCIM_Constant, time=1, value=0}
         ]}}}
      CurveIndex=&Frac_Float(Float=&Add_Float(
        A=&Frac_Float(Float=&Add_Float(A=&Multiply_Float(A=$Particles.PartitionParticles.Normalized Index In Partition, B=-0.25), B=1))
        B=&Frac_Float(Float=&Multiply_Float(
          A=&Add_Float(
           A=&self:Pixelate(Input:NiagaraFloat=$Emitter.Age, Pixel Value:NiagaraFloat=&RandomRangeFloat(Minimum=24, Maximum=8))
           B=$Emitter.Time Offset Random
          )
          B=0.35
         ))
       ))
     )
    ))
   &ScaleColor(
    Scale RGB=&VectorFromFloat(Value=&FloatFromCurve(FloatCurve=DataInterface() {propertyValues {Curve {keys [<at 0.1 100>,<at 0.15 1>,<at 1 1>]}}}))
    ScaleA=true
    ScaleRGB=true
   ) {module=ScaleColor001}
   &FN_Light_Attributes(
    Color=&MultiplyLinearColor(
     Linear Color=$Particles.Initial.Color
     Scale Factor=&MakeFloat4FromVectorAndFloat("Vector (RGB)"=&VectorFromFloat(Value=1.5), "Float (Alpha)"=1)
    )
    Create Color Attribute=true
    Create Exponent Attribute=true
    Create Position Attribute=true
    Create Radius Attribute=true
    Enable Particle Light Enable=true
    Exponent=3
    Light Enabled:NiagaraBool=&SetBoolByIntComparison(A=0, B=$Particles.PartitionParticles.Index In Partition, Comparison Type=::NewEnumerator4)
    Position=&AddVectorToPosition(
     Position=&SimulationPosition()
     Vector=&TransformVector(Vector=<15 5 30>, Destination Space=::Simulation)
    )
    Radius=55
    Scale Color=&FloatFromCurve(FloatCurve=DataInterface() {propertyValues {Curve {keys [
         <at 0 1 RCIM_Cubic>
         {
          arriveTangent=-9.8572
          arriveTangentWeight=1.7722
          interpMode=RCIM_Cubic
          leaveTangentWeight=4.1352
          tangentMode=RCTM_Break
          time=0.3
          value=0
         }
         <at 1 0>
        ]}}})
   ) {enabled=false}
  ]
  rendererClasses [
   &NiagaraRibbonRendererProperties
   &NiagaraSpriteRendererProperties
   &NiagaraLightRendererProperties
  ]
  renderers [
   {rendererClass=&NiagaraRibbonRendererProperties}
   {rendererClass=&NiagaraSpriteRendererProperties, rendererIndex=1}
   {rendererClass=&NiagaraLightRendererProperties, rendererIndex=2}
  ]
 }
 Lightning_Blade {
  emitterUpdateScript [
   &EmitterState(Enable Distance Culling=true, MaxDistance=2500, Module.bUseMaxDistance=true, Module.bUseMinDistance=false, Scalability Mode=::NewEnumerator1)
   &SpawnRate(Spawn Probability=0.75, SpawnGroup=1, SpawnRate=3, Use Spawn Probability=true) {module=SpawnRate001}
   &SpawnRate(Spawn Probability=0.75, SpawnRate=5, Use Spawn Probability=true)
   &SpawnRate(SpawnRate=10, Use Spawn Probability=false) {enabled=false, module=SpawnRate002}
  ]
  particleSpawnScript [
   &InitializeParticle(
    Lifetime=&RandomRangeFloat(Minimum=0.2, Maximum=0.35)
    Mesh Renderer Array Visibility Mode=::NewEnumerator3
    Mesh Renderer Info=DataInterface() {propertyValues {MeshRenderer=&self:Lightning_Blade_0.NiagaraMeshRendererProperties_0}}
    Mesh Scale Max=<0.55 0.55 0.55>
    Mesh Scale Min=<0.8 0.8 0.8>
    Mesh Scale Mode=::NewEnumerator2
    Sprite Size Mode=::NewEnumerator1
    Uniform Sprite Size Max=3
    Uniform Sprite Size Min=2
    UsePositionOffset=false
    Write Material Random=true
   )
   &DynamicMaterialParameters(
    Index 0 Param 1=2
    Index 0 Param 2=&RandomRangeFloat(Minimum=0.65, Maximum=0.75)
    Index 0 Param 4=0.25
    Module.Param0WriteEnabled=true
    Module.Param1WriteEnabled=true
    Module.Param2WriteEnabled=true
    Module.Param3WriteEnabled=true
   )
   &InitialMeshOrientation(Use Rotation Vector=false)
   &ShapeLocation(
    Box Size=&MaskVectorBySpawnGroup(Masked Vector=<0 0 35>, Passthrough Vector=<0 0 0>)
    MaskSpawnGroup=false
    Offset=&MaskVectorBySpawnGroup(Masked Vector=<0 42 45>, Passthrough Vector=<0 0 0>)
    Offset Mode=::NewEnumerator0
    Override Seed=false
    Rotation Angle Type:ENiagara_AngleInput=::NewEnumerator0
    Rotation Mode=::NewEnumerator2
    Shape Primitive=::NewEnumerator2
    Surface Only=false
    UseOwnerScale=true
    Yaw / Pitch / Roll=<0 0 42>
   )
   &ShapeLocation(
    MaskSpawnGroup=false
    Offset=&MaskVectorBySpawnGroup(Masked Vector=<0 0 0>, Passthrough Vector=<1 -7 52>)
    Offset Mode=::NewEnumerator0
    Override Seed=false
    Rotation Angle Type:ENiagara_AngleInput=::NewEnumerator0
    Rotation Mode=::NewEnumerator2
    Sphere Radius=&MaskFloatBySpawnGroup(Masked Float=0, Passthrough Float=5)
    UseOwnerScale=true
    Yaw / Pitch / Roll=<0 0 42>
   ) {module=ShapeLocation001}
   &CameraOffset(Camera Offset Amount=2)
  ]
  particleUpdateScript [
   &ParticleState()
   &ScaleMeshSize(Scale Factor=&VectorFromFloat(Value=1))
   &ScaleColor(
    Scale Alpha=&FloatFromCurve(FloatCurve=DataInterface() {propertyValues {Curve {keys [<at 0.5 1 RCIM_Cubic>,<at 1 0 RCIM_Cubic>]}}}, Scale Curve=2)
    ScaleA=true
    ScaleRGB=true
   )
   &DynamicMaterialParameters(
    Index 0 Param 1=&RandomRangeFloat(Minimum=1.25, Maximum=1.75)
    Index 0 Param 3=$Particles.NormalizedAge
    Index 0 Param 4=0.25
    Module.Param0WriteEnabled=true
    Module.Param1WriteEnabled=false
    Module.Param2WriteEnabled=true
    Module.Param3WriteEnabled=true
   ) {module=DynamicMaterialParameters001}
   &FN_Sprite_Attributes(
    Camera Offset=12
    Color=&MultiplyLinearColor(
     Linear Color=<rgba 0.4366 0.9611 2 0.12>
     Scale Factor=&MakeFloat4FromVectorAndFloat(
      "Vector (RGB)"=<1 1 1>
      "Float (Alpha)"=&FloatFromCurve(
       FloatCurve=DataInterface() {propertyValues {Curve {keys [{interpMode=RCIM_Cubic, tangentMode=RCTM_User, time=0, value=1}, {interpMode=RCIM_Cubic, tangentMode=RCTM_User, time=1, value=0}]}}}
       Scale Curve=&FNTimeOfDayScalar(Night Scale=3, Morning Scale=5, Day Scale=5, Evening Scale=5)
      )
     )
    )
    Create Camera Offset Attribute=true
    Create Color Attribute=true
    Create SpriteSize Attribute=true
    SpriteSize=&Multiply_Vector2D_ByFloat(Float=15, Vector2D=<1 1>)
    Write to Module Namespace=true
   )
   &FN_Light_Attributes(
    Color=$Particles.Color
    Create Color Attribute=true
    Create Exponent Attribute=true
    Create Position Attribute=true
    Create Radius Attribute=true
    Enable Particle Light Enable=true
    Exponent=1
    Light Enabled=true
    Position=&AddVectorToPosition(Position=$Particles.Position, Vector=<0 0 10>)
    Radius=20
    Scale Color=&FloatFromCurve(
     FloatCurve=DataInterface() {propertyValues {Curve {keys [
         <at 0 0 RCIM_Cubic>
         {arriveTangent=0.875, interpMode=RCIM_Cubic, leaveTangent=0.875, time=0.1, value=0.9}
         {arriveTangent=-0.2131, interpMode=RCIM_Cubic, leaveTangent=-0.2131, tangentMode=RCTM_User, time=0.4, value=0.35}
         {arriveTangent=-0.5833, interpMode=RCIM_Cubic, leaveTangent=-0.5833, time=0.7, value=0.5}
         <at 1 0 RCIM_Cubic>
        ]}}}
     Scale Curve=0.01
    )
   )
  ]
  rendererClasses [
   &NiagaraMeshRendererProperties
   &NiagaraSpriteRendererProperties
   &NiagaraLightRendererProperties
  ]
  renderers [
   {rendererClass=&NiagaraMeshRendererProperties, rendererIndex=0}
   {rendererClass=&NiagaraSpriteRendererProperties, rendererIndex=1}
   {rendererClass=&NiagaraLightRendererProperties, rendererIndex=2}
  ]
 }
 Lightning_Blade001 {
  emitterUpdateScript [
   &EmitterState(Enable Distance Culling=true, MaxDistance=2500, Module.bUseMaxDistance=true, Module.bUseMinDistance=false, Scalability Mode=::NewEnumerator1)
   &SpawnRate(Spawn Probability=0.3333, SpawnGroup=1, SpawnRate=4, Use Spawn Probability=true) {module=SpawnRate001}
  ]
  particleSpawnScript [
   &InitializeParticle(
    Color=<rgba 2 52.9284 100 1>
    Lifetime=&RandomRangeFloat(Minimum=0.2, Maximum=0.7)
    Mesh Renderer Array Visibility Mode=::NewEnumerator3
    Mesh Renderer Info=DataInterface() {propertyValues {MeshRenderer=&self:Lightning_Blade001_1.NiagaraMeshRendererProperties_0}}
    Mesh Scale Max=<0.55 0.55 0.55>
    Mesh Scale Min=<0.8 0.8 0.8>
    Mesh Scale Mode=::NewEnumerator2
    Sprite Size Mode=::NewEnumerator1
    Uniform Sprite Size Max=3
    Uniform Sprite Size Min=2
    UsePositionOffset=false
    Write Material Random=true
   )
   &DynamicMaterialParameters(
    Index 0 Param 1=2
    Index 0 Param 2=&RandomRangeFloat(Minimum=0.65, Maximum=0.75)
    Index 0 Param 4=0.25
    Module.Param0WriteEnabled=true
    Module.Param1WriteEnabled=true
    Module.Param2WriteEnabled=true
    Module.Param3WriteEnabled=true
   )
   &InitialMeshOrientation(Use Rotation Vector=false)
   &ShapeLocation(
    Box Size=<0 35 0>
    MaskSpawnGroup=false
    Offset=<0 20 52>
    Offset Mode=::NewEnumerator0
    Override Seed=false
    Rotation Angle Type:ENiagara_AngleInput=::NewEnumerator0
    Rotation Mode=::NewEnumerator2
    Shape Primitive=::NewEnumerator2
    Surface Only=false
    UseOwnerScale=true
   )
   &CameraOffset(Camera Offset Amount=2)
  ]
  particleUpdateScript [
   &ParticleState()
   &ScaleMeshSize(Scale Factor=&VectorFromFloat(Value=1))
   &ScaleColor(
    Scale Alpha=&FloatFromCurve(
     FloatCurve=DataInterface() {propertyValues {Curve {keys [{arriveTangent=-1, interpMode=RCIM_Cubic, time=0.5, value=1}, <at 1 0 RCIM_Cubic>]}}}
     Scale Curve=2
    )
    Scale RGB=&VectorFromFloat(Value=&FloatFromCurve(FloatCurve=DataInterface() {propertyValues {Curve {keys [<at 0 100>,<at 0.1 1>,<at 1 0>]}}}))
    ScaleA=true
    ScaleRGB=true
   )
   &UpdateMeshOrientation(
    Rotation Rate=&RandomRangeFloat(Minimum=0, Maximum=1)
    Rotation Vector=&RandomVector(Vector Scale=1)
   )
   &DynamicMaterialParameters(
    Index 0 Param 1=&RandomRangeFloat(Minimum=2, Maximum=5)
    Index 0 Param 3=$Particles.NormalizedAge
    Index 0 Param 4=&RandomRangeFloat(Minimum=0.5, Maximum=1)
    Module.Param0WriteEnabled=true
    Module.Param1WriteEnabled=false
    Module.Param2WriteEnabled=true
    Module.Param3WriteEnabled=true
   ) {module=DynamicMaterialParameters001}
   &FN_Sprite_Attributes(
    Camera Offset=15
    Color=&MultiplyLinearColor(
     Linear Color=<rgba 0.6549 1.4416 3 0.5>
     Scale Factor=&MakeFloat4FromVectorAndFloat(
      "Vector (RGB)"=<1 1 1>
      "Float (Alpha)"=&FloatFromCurve(
       FloatCurve=DataInterface() {propertyValues {Curve {keys [{interpMode=RCIM_Cubic, tangentMode=RCTM_User, time=0, value=1}, {interpMode=RCIM_Cubic, tangentMode=RCTM_User, time=1, value=0}]}}}
       Scale Curve=&FNTimeOfDayScalar(Night Scale=1, Morning Scale=3, Day Scale=3, Evening Scale=3)
      )
     )
    )
    Create Camera Offset Attribute=true
    Create Color Attribute=true
    Create SpriteSize Attribute=true
    SpriteSize=&Multiply_Vector2D_ByFloat(Float=22, Vector2D=<1 1>)
    Write to Module Namespace=true
   )
   &FN_Light_Attributes(
    Color=$Particles.Color
    Create Color Attribute=true
    Create Exponent Attribute=true
    Create Position Attribute=true
    Create Radius Attribute=true
    Enable Particle Light Enable=true
    Exponent=1
    Light Enabled=true
    Position=&AddVectorToPosition(Position=$Particles.Position, Vector=<0 0 10>)
    Radius=20
    Scale Color=&FloatFromCurve(
     FloatCurve=DataInterface() {propertyValues {Curve {keys [
         <at 0 0 RCIM_Cubic>
         {arriveTangent=0.875, interpMode=RCIM_Cubic, leaveTangent=0.875, time=0.1, value=0.9}
         {arriveTangent=-0.2131, interpMode=RCIM_Cubic, leaveTangent=-0.2131, tangentMode=RCTM_User, time=0.4, value=0.35}
         {arriveTangent=-0.5833, interpMode=RCIM_Cubic, leaveTangent=-0.5833, time=0.7, value=0.5}
         <at 1 0 RCIM_Cubic>
        ]}}}
     Scale Curve=0.05
    )
   )
   &ScaleColor(
    Scale Alpha=&Lerp_Float(
     A=0
     B=1
     Alpha=&UniformAOrBFloat(
      A=&Sine(
       Normalized Angle=&Add_Float(A=$Emitter.Age, B=&RandomRangeFloat(Minimum=0, Maximum=100))
       Period=&RandomRangeFloat(Minimum=0.025, Maximum=0.1)
       Scale=0.5
       Bias=0.5
      )
      Distribution Weight=0.6
     )
    )
    Scale RGB=&VectorFromFloat(Value=1)
    ScaleA=true
    ScaleRGB=true
   ) {module=ScaleColor001}
  ]
  rendererClasses [
   &NiagaraMeshRendererProperties
   &NiagaraSpriteRendererProperties
   &NiagaraLightRendererProperties
  ]
  renderers [
   {rendererClass=&NiagaraMeshRendererProperties, rendererIndex=0}
   {rendererClass=&NiagaraSpriteRendererProperties, rendererIndex=1}
   {rendererClass=&NiagaraLightRendererProperties, rendererIndex=2}
  ]
 }
 Lightning_Blade002 {
  bEnabled=false
  emitterUpdateScript [
   &EmitterState(Enable Distance Culling=true, MaxDistance=2500, Module.bUseMaxDistance=true, Module.bUseMinDistance=false, Scalability Mode=::NewEnumerator1)
   &SpawnRate(SpawnGroup=1, SpawnRate=3, Use Spawn Probability=false) {module=SpawnRate001}
   &SpawnRate(SpawnRate=6, Use Spawn Probability=false)
  ]
  particleSpawnScript [
   &InitializeParticle(
    Color=<rgba 0.66 17.4664 33 1>
    Lifetime=&RandomRangeFloat(Minimum=0.15, Maximum=0.35)
    Mesh Renderer Array Visibility Mode=::NewEnumerator3
    Mesh Renderer Info=DataInterface() {propertyValues {MeshRenderer=&self:Lightning_Blade002_1.NiagaraMeshRendererProperties_0}}
    Mesh Scale Max=<0.55 0.55 0.55>
    Mesh Scale Min=<0.8 0.8 0.8>
    Mesh Scale Mode=::NewEnumerator2
    Sprite Size Mode=::NewEnumerator1
    Uniform Sprite Size Max=3
    Uniform Sprite Size Min=2
    UsePositionOffset=false
    Write Material Random=true
   )
   &DynamicMaterialParameters(
    Index 0 Param 1=2
    Index 0 Param 2=&RandomRangeFloat(Minimum=0.65, Maximum=0.75)
    Index 0 Param 4=0.25
    Module.Param0WriteEnabled=true
    Module.Param1WriteEnabled=true
    Module.Param2WriteEnabled=true
    Module.Param3WriteEnabled=true
   )
   &InitialMeshOrientation(Use Rotation Vector=false)
   &ShapeLocation(
    Box Size=&MaskVectorBySpawnGroup(Masked Vector=<0 0 35>, Passthrough Vector=<0 0 0>)
    MaskSpawnGroup=false
    Offset=&MaskVectorBySpawnGroup(Masked Vector=<0 42 45>, Passthrough Vector=<0 0 0>)
    Offset Mode=::NewEnumerator0
    Override Seed=false
    Rotation Angle Type:ENiagara_AngleInput=::NewEnumerator0
    Rotation Mode=::NewEnumerator2
    Shape Primitive=::NewEnumerator2
    Surface Only=false
    UseOwnerScale=true
    Yaw / Pitch / Roll=<0 0 42>
   )
   &ShapeLocation(
    MaskSpawnGroup=false
    Offset=&MaskVectorBySpawnGroup(Masked Vector=<0 0 0>, Passthrough Vector=<1 -7 52>)
    Offset Mode=::NewEnumerator0
    Override Seed=false
    Rotation Angle Type:ENiagara_AngleInput=::NewEnumerator0
    Rotation Mode=::NewEnumerator2
    Sphere Radius=&MaskFloatBySpawnGroup(Masked Float=0, Passthrough Float=5)
    UseOwnerScale=true
    Yaw / Pitch / Roll=<0 0 42>
   ) {module=ShapeLocation001}
   &CameraOffset(Camera Offset Amount=2)
  ]
  particleUpdateScript [
   &ParticleState()
   &UpdateMeshOrientation(
    Rotation Rate=&RandomRangeFloat(Minimum=0, Maximum=1)
    Rotation Vector=&RandomVector(Vector Scale=1)
   )
   &ScaleMeshSize(Scale Factor=&VectorFromFloat(Value=&Lerp_Float(
      A=1
      B=&RandomRangeFloat(Minimum=0.75, Maximum=1.25)
      Alpha=&FloatFromCurve(FloatCurve=DataInterface() {propertyValues {Curve {keys [
           {
            arriveTangentWeight=0.8339
            interpMode=RCIM_Cubic
            leaveTangentWeight=0.8339
            tangentMode=RCTM_User
            tangentWeightMode=RCTWM_WeightedBoth
            time=0
            value=0
           }
           {arriveTangent=1.9891, interpMode=RCIM_Cubic, leaveTangent=1.9891, tangentMode=RCTM_User, time=1, value=1}
          ]}}})
     )))
   &ScaleColor(
    Scale Alpha=&FloatFromCurve(
     FloatCurve=DataInterface() {propertyValues {Curve {keys [{arriveTangent=-1, interpMode=RCIM_Cubic, time=0.5, value=1}, <at 1 0 RCIM_Cubic>]}}}
     Scale Curve=2
    )
    Scale RGB=&VectorFromFloat(Value=&FloatFromCurve(FloatCurve=DataInterface() {propertyValues {Curve {keys [<at 0 100>,<at 0.1 1>,<at 1 0>]}}}))
    ScaleA=true
    ScaleRGB=true
   )
   &DynamicMaterialParameters(
    Index 0 Param 1=&RandomRangeFloat(Minimum=1.25, Maximum=5)
    Index 0 Param 3=$Particles.NormalizedAge
    Index 0 Param 4=&RandomRangeFloat(Minimum=0.25, Maximum=1)
    Module.Param0WriteEnabled=true
    Module.Param1WriteEnabled=false
    Module.Param2WriteEnabled=true
    Module.Param3WriteEnabled=true
   ) {module=DynamicMaterialParameters001}
   &FN_Sprite_Attributes(
    Camera Offset=15
    Color=&MultiplyLinearColor(
     Linear Color=<rgba 0.6549 1.4416 3 2>
     Scale Factor=&MakeFloat4FromVectorAndFloat(
      "Vector (RGB)"=<1 1 1>
      "Float (Alpha)"=&FloatFromCurve(FloatCurve=DataInterface() {propertyValues {Curve {keys [{interpMode=RCIM_Cubic, tangentMode=RCTM_User, time=0, value=1}, {interpMode=RCIM_Cubic, tangentMode=RCTM_User, time=1, value=0}]}}})
     )
    )
    Create Camera Offset Attribute=true
    Create Color Attribute=true
    Create SpriteSize Attribute=true
    SpriteSize=&Multiply_Vector2D_ByFloat(Float=22, Vector2D=<1 1>)
    Write to Module Namespace=true
   )
   &FN_Light_Attributes(
    Color=$Particles.Color
    Create Color Attribute=true
    Create Exponent Attribute=true
    Create Position Attribute=true
    Create Radius Attribute=true
    Enable Particle Light Enable=true
    Exponent=1
    Light Enabled=true
    Position=&AddVectorToPosition(Position=$Particles.Position, Vector=<0 0 10>)
    Radius=20
    Scale Color=&FloatFromCurve(
     FloatCurve=DataInterface() {propertyValues {Curve {keys [
         <at 0 0 RCIM_Cubic>
         {arriveTangent=0.875, interpMode=RCIM_Cubic, leaveTangent=0.875, time=0.1, value=0.9}
         {arriveTangent=-0.2131, interpMode=RCIM_Cubic, leaveTangent=-0.2131, tangentMode=RCTM_User, time=0.4, value=0.35}
         {arriveTangent=-0.5833, interpMode=RCIM_Cubic, leaveTangent=-0.5833, time=0.7, value=0.5}
         <at 1 0 RCIM_Cubic>
        ]}}}
     Scale Curve=0.05
    )
   )
   &ScaleColor(
    Scale Alpha=&Lerp_Float(
     A=0
     B=1
     Alpha=&UniformAOrBFloat(
      A=&Sine(
       Normalized Angle=&Add_Float(A=$Emitter.Age, B=&RandomRangeFloat(Minimum=0, Maximum=100))
       Period=&RandomRangeFloat(Minimum=0.025, Maximum=0.1)
       Scale=0.5
       Bias=0.5
      )
      Distribution Weight=0.65
     )
    )
    Scale RGB=&VectorFromFloat(Value=1)
    ScaleA=true
    ScaleRGB=true
   ) {module=ScaleColor001}
  ]
  rendererClasses [
   &NiagaraMeshRendererProperties
   &NiagaraSpriteRendererProperties
   &NiagaraLightRendererProperties
  ]
  renderers [
   {rendererClass=&NiagaraMeshRendererProperties, rendererIndex=0}
   {rendererClass=&NiagaraSpriteRendererProperties, rendererIndex=1}
   {rendererClass=&NiagaraLightRendererProperties, rendererIndex=2}
  ]
 }
}
summary {
 systemName=NS_ArcticIceBlue_Idle
 emitters [
  {
   emitterName=Lightning_Blade
   rendererClasses [
    &NiagaraMeshRendererProperties
    &NiagaraSpriteRendererProperties
    &NiagaraLightRendererProperties
   ]
  }
  {emitterName=Glow001, rendererClasses [&NiagaraSpriteRendererProperties, &NiagaraLightRendererProperties]}
  {
   emitterName=Lightning_Blade001
   rendererClasses [
    &NiagaraMeshRendererProperties
    &NiagaraSpriteRendererProperties
    &NiagaraLightRendererProperties
   ]
  }
  {
   bEnabled=false
   emitterName=Lightning_Blade002
   rendererClasses [
    &NiagaraMeshRendererProperties
    &NiagaraSpriteRendererProperties
    &NiagaraLightRendererProperties
   ]
  }
  {
   emitterName=Impact_Lightning
   rendererClasses [
    &NiagaraRibbonRendererProperties
    &NiagaraSpriteRendererProperties
    &NiagaraLightRendererProperties
   ]
  }
 ]
 userVariables []
}
