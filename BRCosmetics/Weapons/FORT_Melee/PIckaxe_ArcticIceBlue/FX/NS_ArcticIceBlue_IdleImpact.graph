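# Niagara effect for the Arctic Ice Blue pickaxe (idle/impact by name) that draws a flickering blue electric arc.
# Impact_Lightning bursts particles into a ribbon per loop, placed by index-driven curves on a cylinder, with a sprite renderer too.
# Updates jitter positions, pulse color, and gate ribbon width through a stepped on/off curve scrolled by pixelated emitter age for stuttering segments.
# The ReF component-renderer emitter and an alternate cylinder path are disabled.
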NiagaraSystem /BRCosmetics/Weapons/FORT_Melee/PIckaxe_ArcticIceBlue/FX/NS_ArcticIceBlue_IdleImpact
dependencies {usedDataInterfaces [&NiagaraDataInterfaceCurve], usedDynamicInputs [], usedModules [&SystemState]}
emitters {
 Impact_Lightning {
  emitterSpawnScript [/SetVariables(Emitter.New Curve for Floats=DataInterface() {propertyValues {Curve {keys [
        {interpMode=RCIM_Constant, leaveTangentWeight=0.003, time=0, value=0}
        {arriveTangentWeight=0.003, interpMode=RCIM_Constant, leaveTangentWeight=0.0092, time=0.0091, value=1}
        {arriveTangentWeight=0.0092, interpMode=RCIM_Constant, leaveTangentWeight=0.0066, time=0.0366, value=0}
        {arriveTangentWeight=0.0066, interpMode=RCIM_Constant, leaveTangentWeight=0.002, time=0.0566, value=1}
        {arriveTangentWeight=0.002, interpMode=RCIM_Constant, leaveTangentWeight=0.0051, time=0.0627, value=0}
        {arriveTangentWeight=0.0051, interpMode=RCIM_Constant, leaveTangentWeight=0.0118, time=0.078, value=1}
        {arriveTangentWeight=0.0118, interpMode=RCIM_Constant, leaveTangentWeight=0.0107, time=0.1133, value=0}
        {arriveTangentWeight=0.0107, interpMode=RCIM_Constant, leaveTangentWeight=0.0077, time=0.1454, value=1}
        {arriveTangentWeight=0.0077, interpMode=RCIM_Constant, leaveTangentWeight=0.003, time=0.1684, value=0}
        {arriveTangentWeight=0.003, interpMode=RCIM_Constant, leaveTangentWeight=0.0092, time=0.1775, value=1}
        {arriveTangentWeight=0.0092, interpMode=RCIM_Constant, leaveTangentWeight=0.0112, time=0.2051, value=0}
        {arriveTangentWeight=0.0112, interpMode=RCIM_Constant, leaveTangentWeight=0.0264, time=0.2386, value=1}
        {arriveTangentWeight=0.0264, interpMode=RCIM_Constant, leaveTangentWeight=0.0309, time=0.3178, value=0}
        {arriveTangentWeight=0.0309, interpMode=RCIM_Constant, leaveTangentWeight=0.0307, time=0.4104, value=1}
        {arriveTangentWeight=0.1965, interpMode=RCIM_Constant, leaveTangentWeight=0.141, time=0.5027, value=0}
        {arriveTangentWeight=0.2939, interpMode=RCIM_Constant, leaveTangentWeight=0.0404, time=0.5689, value=0}
        {arriveTangentWeight=0.003, interpMode=RCIM_Constant, leaveTangentWeight=0.0092, time=0.578, value=1}
        {arriveTangentWeight=0.1715, interpMode=RCIM_Constant, leaveTangentWeight=0.1623, time=0.6055, value=0}
        {arriveTangentWeight=0.21, interpMode=RCIM_Constant, leaveTangentWeight=0.1236, time=0.6315, value=0}
        {arriveTangentWeight=0.0051, interpMode=RCIM_Constant, leaveTangentWeight=0.0118, time=0.6468, value=1}
        {arriveTangentWeight=0.1308, interpMode=RCIM_Constant, leaveTangentWeight=0.2039, time=0.6822, value=0}
        {arriveTangentWeight=0.2867, interpMode=RCIM_Constant, leaveTangentWeight=0.0474, time=0.7373, value=0}
        {arriveTangentWeight=0.003, interpMode=RCIM_Constant, leaveTangentWeight=0.0092, time=0.7464, value=1}
        {arriveTangentWeight=0.0092, interpMode=RCIM_Constant, leaveTangentWeight=0.0112, time=0.7739, value=0}
        {arriveTangentWeight=0.0112, interpMode=RCIM_Constant, leaveTangentWeight=0.0264, time=0.8075, value=1}
        {arriveTangentWeight=0.0264, interpMode=RCIM_Constant, leaveTangentWeight=0.003, time=0.8867, value=0}
        {arriveTangentWeight=0.003, interpMode=RCIM_Constant, leaveTangentWeight=0.0092, time=0.8958, value=1}
        {arriveTangentWeight=0.1715, interpMode=RCIM_Constant, leaveTangentWeight=0.1623, time=0.9233, value=0}
        {arriveTangentWeight=0.2587, interpMode=RCIM_Constant, leaveTangentWeight=0.0748, time=0.9493, value=0}
        {arriveTangentWeight=0.0025, interpMode=RCIM_Constant, leaveTangentWeight=0.0144, time=0.9568, value=1}
        {arriveTangentWeight=0.0144, interpMode=RCIM_Constant, time=1, value=0}
       ]}}}) {isSetParameters=true}]
  emitterUpdateScript [
   &EmitterState()
   /SetVariables(
    Emitter.Rand1:NiagaraFloat=&RandomRangeFloat(Minimum=0, Maximum=1, Evaluation Type=::NewEnumerator1)
    Emitter.Rand2:NiagaraFloat=&RandomRangeFloat(Minimum=0, Maximum=1, Evaluation Type=::NewEnumerator1)
   ) {isSetParameters=true}
   /SetVariables(Emitter.Electric Particles=25) {isSetParameters=true}
   &SpawnBurst_Instantaneous(Spawn Count=$Emitter.Electric Particles, Use Spawn Probability=false, UseLoopCountLimit=false)
   /SetVariables(Emitter.Time Offset Random:NiagaraFloat=&RandomRangeFloat(Minimum=0, Maximum=1)) {isSetParameters=true}
  ]
  particleSpawnScript [
   &PartitionParticles(
    Normalized Index In Partition Includes One=true
    Particles Per Partition:NiagaraInt32=$Emitter.Electric Particles
    Partition Count Mode:ENiagara_InfiniteLoopDuration=::NewEnumerator1
    Partition Mode:ENiagara_PartitionMode=::NewEnumerator1
    Write Partition Index to Ribbon Link Order:ENiagara_AttributeSamplingApplyOutput=::NewEnumerator0
    Write Partition to Ribbon ID=::NewEnumerator0
   )
   &InitializeParticle(
    Color=&RandomRangeLinearColor(Minimum=<rgba 0 4.1574 8 1>, Maximum=<rgba 0.2 0.6157 1 1>)
    Ribbon Width=&Multiply_Float(A=&RandomRangeFloat(Minimum=20, Maximum=15), B=0.6)
    Ribbon Width Mode=::NewEnumerator1
    Sprite Size Mode=::NewEnumerator3
    Uniform Sprite Size=1
    UsePositionOffset=false
    Write Material Random=true
   )
   /SetVariables(
    Particles.RibbonID=/Niagara/DynamicInputs/NiagaraID/MakeNiagaraIDFromInt(Index:NiagaraInt32=$Emitter.LoopCount, Acquire Tag:NiagaraInt32=0)
    Particles.link order:NiagaraFloat=&MakeFloatFromInt(Int=&ReturnParticleIDAttributes())
   ) {isSetParameters=true}
   &CylinderLocation(
    Clamp UV Range=false
    Cylinder Distribution:ENiagara_CylinderMode=::NewEnumerator2
    Cylinder Height=150
    Cylinder Height Midpoint=0.55
    Cylinder Origin:NiagaraPosition=/Niagara/DynamicInputs/Helpers/SimulationPosition()
    Cylinder Radius=11
    MaskSpawnGroup=false
    Offset=<1 7 0>
    Offset Mode:ENiagara_OffsetMode=::NewEnumerator0
    Radius Position:NiagaraFloat=&FloatFromCurve(
     FloatCurve=DataInterface() {propertyValues {Curve {keys [
         {interpMode=RCIM_Cubic, tangentMode=RCTM_User, time=0, value=0.4}
         {
          arriveTangent=0.672
          arriveTangentWeight=0.1067
          interpMode=RCIM_Cubic
          leaveTangentWeight=0.3202
          tangentMode=RCTM_Break
          time=0.5
          value=0.4236
         }
         {interpMode=RCIM_Cubic, tangentMode=RCTM_User, time=0.8, value=0.6099}
         <at 0.8394 0.604>
         <at 0.8655 1.3393>
         {arriveTangentWeight=0.1415, interpMode=RCIM_Cubic, leaveTangentWeight=0.1415, tangentMode=RCTM_User, time=0.9, value=0.5328}
         <at 1 0.6132 RCIM_Cubic>
        ]}}}
     CurveIndex=$Particles.PartitionParticles.Normalized Index In Partition
    )
    U Position:NiagaraFloat=&FloatFromCurve(
     FloatCurve=DataInterface() {propertyValues {Curve {keys [{tangentMode=RCTM_User, time=0, value=2}, {tangentMode=RCTM_User, time=1, value=0.5}]}}}
     CurveIndex=$Particles.PartitionParticles.Normalized Index In Partition
    )
    UseOwnerScale=true
    V Position:NiagaraFloat=&FloatFromCurve(
     FloatCurve=DataInterface() {propertyValues {Curve {keys [{arriveTangent=1.5333, time=0, value=0.1001}, <at 0.9146 0.8469>, {tangentMode=RCTM_User, time=1, value=0.85}]}}}
     CurveIndex=$Particles.PartitionParticles.Normalized Index In Partition
    )
   )
   &CylinderLocation(
    Clamp UV Range=false
    Cylinder Distribution:ENiagara_CylinderMode=::NewEnumerator2
    Cylinder Height=150
    Cylinder Height Midpoint=0.55
    Cylinder Origin:NiagaraPosition=&SimulationPosition()
    Cylinder Radius=11
    MaskSpawnGroup=false
    Offset=<1 7 0>
    Offset Mode:ENiagara_OffsetMode=::NewEnumerator0
    Radius Position:NiagaraFloat=&FloatFromCurve(
     FloatCurve=DataInterface() {propertyValues {Curve {keys [
         {interpMode=RCIM_Cubic, tangentMode=RCTM_User, time=0, value=0.4}
         {
          arriveTangent=0.672
          arriveTangentWeight=0.1067
          interpMode=RCIM_Cubic
          leaveTangentWeight=0.3202
          tangentMode=RCTM_Break
          time=0.5
          value=0.4236
         }
         {interpMode=RCIM_Cubic, tangentMode=RCTM_User, time=0.8, value=0.6099}
         <at 0.8394 0.604>
         <at 0.8655 1.3393>
         {arriveTangentWeight=0.1415, interpMode=RCIM_Cubic, leaveTangentWeight=0.1415, tangentMode=RCTM_User, time=0.9, value=0.5328}
         <at 1 1.225 RCIM_Cubic>
        ]}}}
     CurveIndex=$Particles.PartitionParticles.Normalized Index In Partition
    )
    U Position:NiagaraFloat=&FloatFromCurve(
     FloatCurve=DataInterface() {propertyValues {Curve {keys [
         {arriveTangent=6.2373, interpMode=RCIM_Cubic, leaveTangent=6.2373, tangentMode=RCTM_User, time=0, value=0.3596}
         {arriveTangent=2.109, interpMode=RCIM_Cubic, leaveTangent=2.109, time=0.1949, value=1.3086}
         {
          arriveTangentWeight=0.3586
          interpMode=RCIM_Cubic
          leaveTangent=-3.4745
          leaveTangentWeight=0.3541
          tangentMode=RCTM_Break
          time=0.6
          value=1.625
         }
         {tangentMode=RCTM_User, time=1, value=-0.75}
        ]}}}
     CurveIndex=$Particles.PartitionParticles.Normalized Index In Partition
    )
    UseOwnerScale=true
    V Position:NiagaraFloat=&FloatFromCurve(
     FloatCurve=DataInterface() {propertyValues {Curve {keys [
         {arriveTangent=1.5333, interpMode=RCIM_Cubic, leaveTangent=1.5333, tangentMode=RCTM_User, time=0, value=0.1001}
         {arriveTangent=0.6665, interpMode=RCIM_Cubic, leaveTangent=0.6665, time=0.187, value=0.2321}
         <at 0.6 0.5>
         <at 0.7 0.6>
         <at 0.7532 0.6737>
         <at 0.8141 0.7644>
         <at 0.8648 0.8072>
         <at 0.8815 0.8803>
         <at 0.9146 0.8469>
         {tangentMode=RCTM_User, time=1, value=0.8974}
        ]}}}
     CurveIndex=$Particles.PartitionParticles.Normalized Index In Partition
    )
   ) {enabled=false, module=CylinderLocation001}
   &ShapeLocation(
    MaskSpawnGroup=false
    Offset=&MakeVector(
     X=0
     Y=&FloatFromCurve(
      FloatCurve=DataInterface() {propertyValues {Curve {keys [{tangentMode=RCTM_User, time=0, value=0}, <at 0.9 0>, {tangentMode=RCTM_User, time=1, value=1}]}}}
      CurveIndex=$Particles.PartitionParticles.Normalized Index In Partition
      Scale Curve=26.2269
     )
     Z=0
    )
    Offset Mode=::NewEnumerator0
    Override Seed=false
    Sphere Radius=0
    UseOwnerScale=true
   ) {module=ShapeLocation001}
  ]
  particleUpdateScript [
   &ParticleState(Kill Particles When Lifetime Has Elapsed=false, Let Infinitely Lived Particles Die When Emitter Deactivates=true)
   &ScaleColor(
    Scale Alpha=10
    Scale RGB=&VectorFromFloat(Value=&Power(a=&Waveform("Waveform [1]"=::NewEnumerator12, "[1] Frequency"=35, "[1] Amplitude Min/Max"=<0.8 1>), b=5))
    ScaleA=true
    ScaleRGB=true
   )
   &ShapeLocation(
    Evaluation Type:ENiagaraRandomnessEvaluation=::NewEnumerator1
    MaskSpawnGroup=false
    Override Seed=false
    Shape Origin=$Particles.Position
    Sphere Radius=&RandomRangeFloat(Minimum=0, Maximum=2, Evaluation Type=::NewEnumerator1)
    UseOwnerScale=true
   )
   &ScaleRibbonWidth(Ribbon Width Scale=&Multiply_Float(
     A=&Multiply_Float(
      A=&FloatFromCurve(FloatCurve=DataInterface() {propertyValues {Curve {keys [
           <at 0 0>
           {interpMode=RCIM_Cubic, tangentMode=RCTM_User, time=0.05, value=1}
           {arriveTangentWeight=0.1491, interpMode=RCIM_Cubic, leaveTangentWeight=0.2236, tangentMode=RCTM_User, time=0.8, value=1}
           <at 1 0>
          ]}}})
      B=&FloatFromCurve(
       FloatCurve=DataInterface() {propertyValues {Curve {keys [
           {arriveTangent=6.6108, interpMode=RCIM_Cubic, leaveTangent=6.6108, tangentMode=RCTM_User, time=0, value=0}
           {arriveTangent=2, interpMode=RCIM_Cubic, leaveTangent=2, time=0.05, value=0.7}
           {interpMode=RCIM_Cubic, tangentMode=RCTM_User, time=0.5, value=1}
           {arriveTangent=-6.3941, interpMode=RCIM_Cubic, leaveTangent=-6.3941, tangentMode=RCTM_User, time=1, value=0}
          ]}}}
       CurveIndex=$Particles.Initial.PartitionParticles.Normalized Index In Partition
      )
     )
     B=&FloatFromCurve(
      FloatCurve=DataInterface() {propertyValues {Curve {keys [
          {interpMode=RCIM_Constant, leaveTangentWeight=0.003, time=0, value=0}
          {arriveTangentWeight=0.003, interpMode=RCIM_Constant, leaveTangentWeight=0.0092, time=0.0091, value=1}
          {arriveTangentWeight=0.0092, interpMode=RCIM_Constant, leaveTangentWeight=0.0066, time=0.0366, value=0}
          {arriveTangentWeight=0.002, interpMode=RCIM_Constant, leaveTangentWeight=0.0051, time=0.0627, value=0}
          {arriveTangentWeight=0.0051, interpMode=RCIM_Constant, leaveTangentWeight=0.0118, time=0.078, value=1}
          {arriveTangentWeight=0.0118, interpMode=RCIM_Constant, leaveTangentWeight=0.0107, time=0.1133, value=0}
          {arriveTangentWeight=0.0077, interpMode=RCIM_Constant, leaveTangentWeight=0.003, time=0.1684, value=0}
          {arriveTangentWeight=0.003, interpMode=RCIM_Constant, leaveTangentWeight=0.0092, time=0.1775, value=1}
          {arriveTangentWeight=0.0092, interpMode=RCIM_Constant, leaveTangentWeight=0.0112, time=0.2051, value=0}
          {arriveTangentWeight=0.0112, interpMode=RCIM_Constant, leaveTangentWeight=0.0264, time=0.2386, value=1}
          {arriveTangentWeight=0.0264, interpMode=RCIM_Constant, leaveTangentWeight=0.0309, time=0.3178, value=0}
          {arriveTangentWeight=0.1723, interpMode=RCIM_Constant, leaveTangentWeight=0.1666, time=0.4104, value=1}
          {arriveTangentWeight=0.3251, interpMode=RCIM_Constant, leaveTangentWeight=0.0097, time=0.5, value=1}
          {arriveTangentWeight=0.0129, interpMode=RCIM_Constant, leaveTangentWeight=0.3212, time=0.5027, value=0}
          {arriveTangentWeight=0.2939, interpMode=RCIM_Constant, leaveTangentWeight=0.0404, time=0.5689, value=0}
          {arriveTangentWeight=0.003, interpMode=RCIM_Constant, leaveTangentWeight=0.0092, time=0.578, value=1}
          {arriveTangentWeight=0.1715, interpMode=RCIM_Constant, leaveTangentWeight=0.1623, time=0.6055, value=0}
          {arriveTangentWeight=0.21, interpMode=RCIM_Constant, leaveTangentWeight=0.1236, time=0.6315, value=0}
          {arriveTangentWeight=0.0051, interpMode=RCIM_Constant, leaveTangentWeight=0.0118, time=0.6468, value=1}
          {arriveTangentWeight=0.1308, interpMode=RCIM_Constant, leaveTangentWeight=0.2039, time=0.6822, value=0}
          {arriveTangentWeight=0.2867, interpMode=RCIM_Constant, leaveTangentWeight=0.0474, time=0.7373, value=0}
          {arriveTangentWeight=0.003, interpMode=RCIM_Constant, leaveTangentWeight=0.0092, time=0.7464, value=1}
          {arriveTangentWeight=0.0092, interpMode=RCIM_Constant, leaveTangentWeight=0.0112, time=0.7739, value=0}
          {arriveTangentWeight=0.0112, interpMode=RCIM_Constant, leaveTangentWeight=0.0264, time=0.8075, value=1}
          {arriveTangentWeight=0.0264, interpMode=RCIM_Constant, leaveTangentWeight=0.003, time=0.8867, value=0}
          {arriveTangentWeight=0.003, interpMode=RCIM_Constant, leaveTangentWeight=0.0092, time=0.8958, value=1}
          {arriveTangentWeight=0.1715, interpMode=RCIM_Constant, leaveTangentWeight=0.1623, time=0.9233, value=0}
          {arriveTangentWeight=0.2587, interpMode=RCIM_Constant, leaveTangentWeight=0.0748, time=0.9493, value=0}
          {arriveTangentWeight=0.0025, interpMode=RCIM_Constant, leaveTangentWeight=0.0144, time=0.9568, value=1}
          {arriveTangentWeight=0.0144, interpMode=RCIM_Constant, time=1, value=0}
         ]}}}
      CurveIndex=&Frac_Float(Float=&Add_Float(
        A=&Frac_Float(Float=&Add_Float(A=&Multiply_Float(A=$Particles.PartitionParticles.Normalized Index In Partition, B=-0.25), B=1))
        B=&Frac_Float(Float=&Multiply_Float(
          A=&Add_Float(
           A=&self:Pixelate(Input:NiagaraFloat=$Emitter.Age, Pixel Value:NiagaraFloat=&RandomRangeFloat(Minimum=24, Maximum=8))
           B=$Emitter.Time Offset Random
          )
          B=0.35
         ))
       ))
     )
    )) {module=ScaleRibbonWidth001}
  ]
  rendererClasses [&NiagaraRibbonRendererProperties, &NiagaraSpriteRendererProperties]
  renderers [
   {rendererClass=&NiagaraRibbonRendererProperties}
   {rendererClass=&NiagaraSpriteRendererProperties, rendererIndex=1}
  ]
 }
 ReF {
  bEnabled=false
  emitterUpdateScript [
   &EmitterState()
   &SpawnBurst_Instantaneous(Use Spawn Probability=false, UseLoopCountLimit=false)
  ]
  particleSpawnScript [&InitializeParticle(Color Mode=::NewEnumerator0, UsePositionOffset=false, Write Material Random=true)]
  particleUpdateScript [&ParticleState()]
  rendererClasses [&NiagaraComponentRendererProperties]
  renderers [{rendererClass=&NiagaraComponentRendererProperties, rendererIndex=0}]
 }
}
summary {
 systemName=NS_ArcticIceBlue_IdleImpact
 emitters [
  {
   emitterName=Impact_Lightning
   rendererClasses [&NiagaraRibbonRendererProperties, &NiagaraSpriteRendererProperties]
  }
  {bEnabled=false, emitterName=ReF, rendererClasses [&NiagaraComponentRendererProperties]}
 ]
 userVariables [User.PlayTimeline {type=NiagaraBool}]
}
